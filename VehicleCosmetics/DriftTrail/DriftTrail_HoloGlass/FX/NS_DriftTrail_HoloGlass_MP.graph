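# Niagara drift trail cosmetic "HoloGlass": glowing cores, holographic glass shards and triangle bursts emitted from the vehicle's four wheels.
# Core layers three sprite spawn groups, Glass streams random mesh shards backward and emits death events, and a GPU emitter wraps shards around wheels.
# TriangleExplotion scatters short-lived sprites, with its own spawn rate disabled.
# Partitioning maps particles to wheel sockets for per-wheel masking, while user parameters drive painted colors.
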NiagaraSystem /VehicleCosmetics/DriftTrail/DriftTrail_HoloGlass/FX/NS_DriftTrail_HoloGlass_MP
dependencies {
 usedDataInterfaces [
  &NiagaraDataInterfaceColorCurve
  &NiagaraDataInterfaceArrayColor
  &NiagaraDataInterfaceCurve
  &NiagaraDataInterfaceArrayFloat3
  &NiagaraDataInterfaceSkeletalMesh
 ]
 usedDynamicInputs []
 usedModules [
  &self:SystemSpawnScript.SystemScriptSource.SystemScriptGraph.NiagaraNodeAssignment_0.SetVariables_#1
  &FN_UnpackVehicleFlags
  /VehicleCosmetics/FX/Niagara/Modules/VC_SetVehicleSpeedAttribs
  &SystemState
 ]
}
emitters {
 Core {
  emitterUpdateScript [
   &EmitterState()
   &SpawnRate(SpawnRate=24, Use Spawn Probability=false)
   &SpawnRate(SpawnGroup=1, SpawnRate=6, Use Spawn Probability=false) {module=SpawnRate001}
   &SpawnRate(SpawnGroup=2, SpawnRate=60, Use Spawn Probability=false) {module=SpawnRate002}
  ]
  particleSpawnScript [
   &VC_PaintedVariantsOverride() {enabled=false}
   /SetVariables(Particles.RENDERTAG:NiagaraInt32=&SpawnGroup()) {isSetParameters=true}
   &InitializeParticle(
    Color Mode=::NewEnumerator0
    Lifetime=/VehicleCosmetics/FX/Niagara/Modules/MergeEmitters/VC_FloatBySpawnGroup(
     SpawnGroup or ExecIndex:NiagaraInt32=0
     CustomInt:NiagaraInt32=0
     MaximumSpawnGroups:NiagaraInt32=3
     Float SG 0:NiagaraFloat=0.5
     Float SG 1:NiagaraFloat=&RandomRangeFloat(Minimum=0.2, Maximum=0.4)
     Float SG 2:NiagaraFloat=&RandomRangeFloat(Minimum=0.4, Maximum=0.6)
     Float SG 3:NiagaraFloat=0
     Float SG 4:NiagaraFloat=0
     Float SG 5:NiagaraFloat=0
    )
    Sprite Rotation Mode=::NewEnumerator2
    Sprite Size Mode=::NewEnumerator3
    Uniform Sprite Size=&VC_FloatBySpawnGroup(
     SpawnGroup or ExecIndex:NiagaraInt32=0
     CustomInt:NiagaraInt32=0
     MaximumSpawnGroups:NiagaraInt32=3
     Float SG 0:NiagaraFloat=8
     Float SG 1:NiagaraFloat=&Multiply_Float(A=0.6, B=&RandomRangeFloat(Minimum=20, Maximum=50))
     Float SG 2:NiagaraFloat=&Multiply_Float(A=0.2, B=&RandomRangeFloat(Minimum=20, Maximum=50))
     Float SG 3:NiagaraFloat=0
     Float SG 4:NiagaraFloat=0
     Float SG 5:NiagaraFloat=0
    )
    UsePositionOffset=false
    Write Material Random=true
   )
   &PartitionParticles()
   &SkeletalMeshLocation(
    Bone / Socket Index=&Add_Integer(A=$Particles.PartitionParticles.Partition, B=4)
    Bone / Socket Sampling Mode=::NewEnumerator5
    Interpolate Mesh Sampling=false
    MaskSpawnGroup=false
    Mesh Sampling Type=::NewEnumerator6
    Sampled Position Offset=&AddVectorToPosition(Position=<0 0 -20>, Vector=$User.offset)
    Skeletal Mesh=$System.SkeletalMesh
   )
   &AddVelocity(Velocity=&VC_Mask_Vectors(
     FrontEnd Value:Vector3f=<-1500 0 0>
     blsFrontEndPreview:NiagaraBool=false
     SpeedMin Value:Vector3f=<-800 0 0>
     SpeedMax Value:Vector3f=<-1 0 0>
     Speed Alpha:NiagaraFloat=$System.Speed_Alpha
    )) {enabled=false}
   &Color(Color=&MakeCustomLinearColorFromBool(Bool=$User.bIsPainted, True Color=$User.CoreColor, False Color=<rgba 1 0.3 0.01 1>)) {enabled=false}
   /SetVariables(
    Particles.CoreColor:LinearColor=&MakeCustomLinearColorFromBool(Bool=$User.bIsPainted, True Color=$User.CoreColor, False Color=<rgba 1 0.3 0.01 1>)
    Particles.OrbitsColor:LinearColor=&MakeCustomLinearColorFromBool(
     Bool=$User.bIsPainted
     True Color=$User.PaintedColor
     False Color=&SelectLinearColorFromArray(
      Array Sampling Mode=::NewEnumerator0
      Direct Array Index=?
      Color Selection Array=DataInterface() {propertyValues {ColorData [<rgba 1 0.2 1 1>, <rgba 0.2 1 1 1>, <rgba 1 1 0.05 1>, <rgba 0.2 0.05 1 1>]}}
      Random Seed=0
     )
    )
    Particles.SphereColor:LinearColor=&MakeCustomLinearColorFromBool(
     Bool=$User.bIsPainted
     True Color=$User.GlowColor
     False Color=&SelectLinearColorFromArray(
      Array Sampling Mode=::NewEnumerator0
      Direct Array Index=?
      Color Selection Array=DataInterface() {propertyValues {ColorData [<rgba 1 0.2 1 1>, <rgba 0.2 1 1 1>, <rgba 1 1 0.05 1>, <rgba 0.2 0.05 1 1>]}}
      Random Seed=0
     )
    )
   ) {isSetParameters=true}
  ]
  particleUpdateScript [
   &FN_DisableTireFX(
    Mask Index Attribute:NiagaraInt32=$Particles.PartitionParticles.Partition
    Mask Index FL=2
    Mask Index FR=3
    Mask Index RL=0
    Mask Index RR=1
   )
   &ScaleColor(
    Scale Alpha=&FloatFromCurve(FloatCurve=DataInterface() {propertyValues {Curve {keys [<at 0 0.5>,<at 1 0>]}}})
    Scale RGB=&VectorFromFloat(Value=&FloatFromCurve(FloatCurve=DataInterface() {propertyValues {Curve {keys [<at 0 10>,<at 1 2>]}}}, Scale Curve=4))
    ScaleA=true
    ScaleRGB=true
   )
   &ParticleState()
   &ScaleSpriteSize(
    Scale Sprite Size Mode=::NewEnumerator0
    Uniform Scale Factor=&VC_FloatBySpawnGroup(
     SpawnGroup or ExecIndex:NiagaraInt32=0
     CustomInt:NiagaraInt32=0
     MaximumSpawnGroups:NiagaraInt32=3
     Float SG 0:NiagaraFloat=&FloatFromCurve(
      FloatCurve=DataInterface() {propertyValues {Curve {keys [{tangentMode=RCTM_User, time=0, value=0}, {tangentMode=RCTM_User, time=1, value=1}]}}}
      Scale Curve=6
     )
     Float SG 1:NiagaraFloat=&FloatFromCurve(FloatCurve=DataInterface() {propertyValues {Curve {keys [<at 0 0>,<at 0.5 1>,<at 1 0>]}}})
     Float SG 2:NiagaraFloat=&FloatFromCurve(FloatCurve=DataInterface() {propertyValues {Curve {keys [{interpMode=RCIM_Cubic, tangentMode=RCTM_User, time=0.4, value=1}, {interpMode=RCIM_Cubic, tangentMode=RCTM_User, time=1, value=0}]}}})
     Float SG 3:NiagaraFloat=0
     Float SG 4:NiagaraFloat=0
     Float SG 5:NiagaraFloat=0
    )
   )
   &SolveForcesAndVelocity(Clamp Velocity=false, Limit Acceleration=false)
   /SetVariables(
    Particles.CoreColorScale:LinearColor=&VC_DynamicScaleColor(
     Color to Scale:LinearColor=$Particles.Initial.CoreColor
     RGB Scale:Vector3f=&VectorFromFloat(Value=&FloatFromCurve(FloatCurve=DataInterface() {propertyValues {Curve {keys [<at 0 10>,<at 1 2>]}}}, Scale Curve=4))
     Alpha Scale:NiagaraFloat=&FloatFromCurve(FloatCurve=DataInterface() {propertyValues {Curve {keys [<at 0 0.5>,<at 1 0>]}}})
    )
    Particles.OrbitsColorScale:LinearColor=&VC_DynamicScaleColor(
     Color to Scale:LinearColor=$Particles.Initial.OrbitsColor
     RGB Scale:Vector3f=&MakeCustomVectorFromBool(
      Bool=$User.bIsPainted
      True Vector=&VectorFromFloat(Value=&RandomRangeFloat(Minimum=$User.ParticleColorVar, Maximum=1))
      False Vector=&Multiply_VectorByFloat(
       Vector=&VectorFromFloat(Value=&FloatFromCurve(FloatCurve=DataInterface() {propertyValues {Curve {keys [<at 0 20>,<at 0.5 1>]}}}))
       Float=1
      )
     )
     Alpha Scale:NiagaraFloat=&FloatFromCurve(FloatCurve=DataInterface() {propertyValues {Curve {keys [<at 0 0>,<at 0.3 1>,<at 0.7 1>,<at 1 0>]}}})
    )
    Particles.SphereColorScale:LinearColor=&VC_DynamicScaleColor(
     Color to Scale:LinearColor=$Particles.Initial.SphereColor
     RGB Scale:Vector3f=&MakeCustomVectorFromBool(
      Bool=$User.bIsPainted
      True Vector=&VectorFromFloat(Value=&FloatFromCurve(FloatCurve=DataInterface() {propertyValues {Curve {keys [{tangentMode=RCTM_User, time=0, value=5}, {tangentMode=RCTM_User, time=1, value=2.5}]}}}))
      False Vector=&VectorFromFloat(Value=&FloatFromCurve(FloatCurve=DataInterface() {propertyValues {Curve {keys [<at 0 5>,<at 1 0>]}}}, Scale Curve=15))
     )
     Alpha Scale:NiagaraFloat=1
    )
   ) {isSetParameters=true}
   &ShapeLocation(
    MaskSpawnGroup=false
    Offset=&MakeVector(
     X=&Sine(
      Normalized Angle=$Emitter.Age
      Period=&RandomRangeFloat(Minimum=2, Maximum=3)
      Scale=&VC_FloatBySpawnGroup(
       SpawnGroup or ExecIndex:NiagaraInt32=0
       CustomInt:NiagaraInt32=0
       MaximumSpawnGroups:NiagaraInt32=3
       Float SG 0:NiagaraFloat=0
       Float SG 1:NiagaraFloat=0
       Float SG 2:NiagaraFloat=&RandomRangeFloat(Minimum=-40, Maximum=40)
       Float SG 3:NiagaraFloat=0
       Float SG 4:NiagaraFloat=0
       Float SG 5:NiagaraFloat=0
      )
      Bias=&RandomRangeFloat(Minimum=0, Maximum=1)
     )
     Y=&Sine(
      Normalized Angle=$Emitter.Age
      Period=&RandomRangeFloat(Minimum=2, Maximum=3)
      Scale=&VC_FloatBySpawnGroup(
       SpawnGroup or ExecIndex:NiagaraInt32=0
       CustomInt:NiagaraInt32=0
       MaximumSpawnGroups:NiagaraInt32=3
       Float SG 0:NiagaraFloat=0
       Float SG 1:NiagaraFloat=0
       Float SG 2:NiagaraFloat=&RandomRangeFloat(Minimum=-40, Maximum=40)
       Float SG 3:NiagaraFloat=0
       Float SG 4:NiagaraFloat=0
       Float SG 5:NiagaraFloat=0
      )
      Bias=&RandomRangeFloat(Minimum=0, Maximum=1)
     )
     Z=&Sine(
      Normalized Angle=$Emitter.Age
      Period=&RandomRangeFloat(Minimum=2, Maximum=3)
      Scale=&VC_FloatBySpawnGroup(
       SpawnGroup or ExecIndex:NiagaraInt32=0
       CustomInt:NiagaraInt32=0
       MaximumSpawnGroups:NiagaraInt32=3
       Float SG 0:NiagaraFloat=0
       Float SG 1:NiagaraFloat=0
       Float SG 2:NiagaraFloat=&RandomRangeFloat(Minimum=-40, Maximum=40)
       Float SG 3:NiagaraFloat=0
       Float SG 4:NiagaraFloat=0
       Float SG 5:NiagaraFloat=0
      )
      Bias=&RandomRangeFloat(Minimum=0, Maximum=1)
     )
    )
    Offset Mode=::NewEnumerator0
    Override Seed=false
    Shape Origin=$Particles.Initial.Position
    Sphere Radius=&VC_FloatBySpawnGroup(
     SpawnGroup or ExecIndex:NiagaraInt32=0
     CustomInt:NiagaraInt32=0
     MaximumSpawnGroups:NiagaraInt32=3
     Float SG 0:NiagaraFloat=0
     Float SG 1:NiagaraFloat=0
     Float SG 2:NiagaraFloat=10
     Float SG 3:NiagaraFloat=0
     Float SG 4:NiagaraFloat=0
     Float SG 5:NiagaraFloat=0
    )
    UseOwnerScale=true
   )
  ]
  rendererClasses [&NiagaraSpriteRendererProperties]
  renderers [
   {rendererClass=&NiagaraSpriteRendererProperties}
   {rendererClass=&NiagaraSpriteRendererProperties, rendererIndex=1}
   {rendererClass=&NiagaraSpriteRendererProperties, rendererIndex=2}
  ]
 }
 Glass {
  emitterUpdateScript [
   &EmitterState()
   &SpawnRate(SpawnRate=100, Use Spawn Probability=false)
  ]
  particleSpawnScript [
   &VC_PaintedVariantsOverride(
    Black=<rgba 0.06 0.06 0.06 1>
    BurntSienna=<rgba 0.19 0.076 0.019 1>
    Cobalt=<rgba 0.05 0.0833 0.25 1>
    Crimson=<rgba 0.35 0.0525 0.0606 1>
    DefaultColor=&ColorFromCurve(
     ColorCurve=DataInterface() {propertyValues {
       AlphaCurve {keys [<at 1 1 RCIM_Constant>]}
       BlueCurve {keys [<at 0 0.3 RCIM_Constant>,<at 0.25 0.28 RCIM_Constant>,<at 0.5233 0.0055 RCIM_Constant>,<at 0.75 0.45 RCIM_Constant>]}
       GreenCurve {keys [<at 0 0.003 RCIM_Constant>,<at 0.25 0.28 RCIM_Constant>,<at 0.5233 0.2758 RCIM_Constant>,<at 0.75 0.0664 RCIM_Constant>]}
       RedCurve {keys [<at 0 0.3 RCIM_Constant>,<at 0.25 0.0028 RCIM_Constant>,<at 0.5233 0.55 RCIM_Constant>,<at 0.75 0.1932 RCIM_Constant>]}
      }}
     CurveIndex=&RandomRangeFloat(Minimum=0, Maximum=1)
    )
    ForestGreen=<rgba 0.0625 0.25 0.072 1>
    Grey=<rgba 0.12 0.12 0.12 1>
    Lime=<rgba 0.15 0.3 0 1>
    Orange=<rgba 0.35 0.1221 0 1>
    Pink=<rgba 0.3 0.165 0.2509 1>
    Purple=<rgba 0.0933 0 0.28 1>
    Saffron=<rgba 0.4 0.3067 0 1>
    SkyBlue=<rgba 0 0.2329 0.35 1>
    TitaniumWhite=<rgba 0.3 0.3 0.3 1>
   )
   &InitializeParticle(
    Color Mode=::NewEnumerator0
    Lifetime=&MakeCustomFloatFromBool(
     Bool=$User.bIsFrontendPreview
     True Float=&RandomRangeFloat(Minimum=0.1, Maximum=0.15)
     False Float=&RandomRangeFloat(Minimum=0.02, Maximum=0.25)
    )
    Mesh Scale=&Multiply_Vector(
     A=&VectorFromFloat(Value=0.2)
     B=&RandomRangeVector(Minimum=<1 1 1>, Maximum=<1 3 1>)
    )
    Mesh Scale Mode=::NewEnumerator4
    Sprite Rotation Mode=::NewEnumerator2
    UsePositionOffset=false
    Write Material Random=true
   )
   &PartitionParticles()
   &SkeletalMeshLocation(
    Bone / Socket Index=&Add_Integer(A=$Particles.PartitionParticles.Partition, B=4)
    Bone / Socket Sampling Mode=::NewEnumerator5
    Interpolate Mesh Sampling=false
    MaskSpawnGroup=false
    Mesh Sampling Type=::NewEnumerator6
    Sampled Position Offset=&AddVectorToPosition(Position=<-70 0 -20>, Vector=$User.offset)
    Skeletal Mesh=$System.SkeletalMesh
   )
   &AddVelocity(Velocity=&MakeCustomVectorFromBool(Bool=$User.bIsFrontendPreview, True Vector=<-3000 0 0>, False Vector=<-1000 0 0>))
   &ShapeLocation(MaskSpawnGroup=false, Override Seed=false, Sphere Radius=20, UseOwnerScale=true)
   &Color(Color=&MakeCustomLinearColorFromBool(
     Bool=$User.bIsPainted
     True Color=&Multiply_LinearColorByFloat(
      Linear Color=$User.GlassColor
      Float=&RandomRangeFloat(Minimum=$User.MinimumGlassColor, Maximum=$User.MaxGlassColor)
     )
     False Color=&SelectLinearColorFromArray(
      Array Sampling Mode=::NewEnumerator0
      Direct Array Index=?
      Color Selection Array=DataInterface() {propertyValues {ColorData [<rgba 1 0.01 1 1>, <rgba 0.01 1 1 1>, <rgba 1 0.5 0.01 1>, <rgba 0.2 0.01 1 1>]}}
      Random Seed=0
     )
    )) {enabled=false}
   &InitialMeshOrientation(
    Mesh Axis To Orient=<0 0 0>
    Mesh Orientation Mode=::NewEnumerator2
    Rotation=&RandomRangeVector(Minimum=<-1 0 0>, Maximum=<1 0 0>)
    Use Rotation Vector=false
   )
   /SetVariables(Particles.MeshIndex=&RandomRangeInt(Minimum=0, Maximum=3)) {isSetParameters=true}
  ]
  particleUpdateScript [
   &FN_DisableTireFX(
    Mask Index Attribute:NiagaraInt32=$Particles.PartitionParticles.Partition
    Mask Index FL=2
    Mask Index FR=3
    Mask Index RL=0
    Mask Index RR=1
   )
   &ScaleColor(
    Scale Alpha=20
    Scale RGB=&MakeCustomVectorFromBool(
     Bool=$User.bIsPainted
     True Vector=&VectorFromFloat(Value=&Multiply_Float(
       A=&RandomRangeFloat(Minimum=$User.GlassColorVar, Maximum=1)
       B=&FloatFromCurve(FloatCurve=DataInterface() {propertyValues {Curve {keys [
            <at 0 1>
            {interpMode=RCIM_Cubic, tangentMode=RCTM_Break, time=0.5, value=1}
            {interpMode=RCIM_Cubic, tangentMode=RCTM_Break, time=1, value=25}
           ]}}})
      ))
     False Vector=&VectorFromFloat(Value=&RandomRangeFloat(
       Minimum=&FloatFromCurve(FloatCurve=DataInterface() {propertyValues {Curve {keys [<at 0 10>,<at 0.1 1>,<at 0.9 1>,<at 1 100>]}}}, Scale Curve=0.6)
       Maximum=&FloatFromCurve(FloatCurve=DataInterface() {propertyValues {Curve {keys [<at 0 5>,<at 0.1 1>,<at 0.9 1>,<at 1 5>]}}}, Scale Curve=0.8)
      ))
    )
    ScaleA=true
    ScaleRGB=true
   ) {enabled=false}
   &InheritSourceMovement(
    Applied Position Scale=&VectorFromFloat(Value=&FloatFromCurve(FloatCurve=DataInterface() {propertyValues {Curve {keys [<at 0.6 1 RCIM_Cubic>,<at 1 0 RCIM_Cubic>]}}}))
    ApplyForce=false
    ApplyPosition=true
    ApplyPositionFalloff=false
   )
   &ParticleState()
   &DynamicMaterialParameters(Module.Index0Param0Write=true, Module.Index0Param1Write=false, Module.Index0Param2Write=false, Module.Index0Param3Write=false)
   &SolveForcesAndVelocity(Clamp Velocity=false, Limit Acceleration=false)
   &GenerateDeathEvent(Use Event Probability=false)
   &ScaleColor(
    Scale Alpha=20
    Scale RGB=&MakeCustomVectorFromBool(
     Bool=$User.bIsPainted
     True Vector=&VectorFromFloat(Value=&RandomRangeFloat(
       Minimum=&FloatFromCurve(FloatCurve=DataInterface() {propertyValues {Curve {keys [<at 0 2>,<at 0.2 0.8>,<at 0.8 0.8>,<at 0.8 0.8>,<at 1 5>]}}})
       Maximum=&FloatFromCurve(FloatCurve=DataInterface() {propertyValues {Curve {keys [<at 0 3>,<at 0.1 3>,<at 0.75 3>,<at 1 10>]}}})
      ))
     False Vector=&VectorFromFloat(Value=&RandomRangeFloat(
       Minimum=&FloatFromCurve(FloatCurve=DataInterface() {propertyValues {Curve {keys [<at 0 5>,<at 0.2 1.25>,<at 0.9 1.25>,<at 1 2>]}}})
       Maximum=&FloatFromCurve(FloatCurve=DataInterface() {propertyValues {Curve {keys [<at 0 8>,<at 0.3 3>,<at 0.9 3>,<at 1 5>]}}})
      ))
    )
    ScaleA=true
    ScaleRGB=true
   ) {module=ScaleColor002}
  ]
  rendererClasses [&NiagaraMeshRendererProperties, &NiagaraSpriteRendererProperties]
  renderers [
   {rendererClass=&NiagaraMeshRendererProperties, rendererIndex=0}
   {rendererClass=&NiagaraSpriteRendererProperties, rendererIndex=1}
  ]
 }
 TriangleExplotion {
  emitterUpdateScript [
   &EmitterState()
   &SpawnRate(SpawnRate=8000, Use Spawn Probability=false) {enabled=false}
  ]
  particleSpawnScript [
   &VC_PaintedVariantsOverride() {enabled=false}
   &InitializeParticle(
    Color Mode=::NewEnumerator0
    Lifetime=&RandomRangeFloat(Minimum=0.2, Maximum=0.1)
    Position Mode=::NewEnumerator0
    Sprite Rotation Mode=::NewEnumerator2
    Sprite Size Mode=::NewEnumerator3
    Write Material Random=true
   )
   &PartitionParticles()
   &ShapeLocation(MaskSpawnGroup=false, Override Seed=false, Sphere Radius=10, UseOwnerScale=true)
   &SkeletalMeshLocation(
    Bone / Socket Index=&Add_Integer(A=$Particles.PartitionParticles.Partition, B=4)
    Bone / Socket Sampling Mode=::NewEnumerator5
    Interpolate Mesh Sampling=false
    MaskSpawnGroup=false
    Mesh Sampling Type=::NewEnumerator6
    Sampled Position Offset=&AddVectorToPosition(Position=<0 0 0>, Vector=&AddVectorToPosition(Position=<-400 0 -20>, Vector=$User.offset))
    Sampled Position Offset Coordinate Space:ENiagaraSkelMeshTransforms=::NewEnumerator2
    Skeletal Mesh=$System.SkeletalMesh
   )
   &AddVelocity(
    Velocity=&RandomRangeVector(Minimum=<1000 1000 1000>, Maximum=<-1000 -1000 -1000>)
    Velocity Speed Scale=0.5
   )
   &Color(Color=<rgba 1 1 1 0>)
  ]
  particleUpdateScript [
   &FN_DisableTireFX(
    Mask Index Attribute:NiagaraInt32=$Particles.PartitionParticles.Partition
    Mask Index FL=2
    Mask Index FR=3
    Mask Index RL=0
    Mask Index RR=1
   )
   &ScaleColor(
    Scale RGB=&MakeCustomVectorFromBool(
     Bool=$User.bIsPainted
     True Vector=&VectorFromFloat(Value=&RandomRangeFloat(Minimum=$User.ParticleColorVar, Maximum=&Multiply_Float(A=1, B=1)))
     False Vector=&Multiply_VectorByFloat(
      Vector=&VectorFromFloat(Value=&RandomRangeFloat(Minimum=1, Maximum=3))
      Float=0.8
     )
    )
    ScaleA=true
    ScaleRGB=true
   )
   &Color(
    Color=&MakeCustomLinearColorFromBool(
     Bool=$User.bIsPainted
     True Color=$User.PaintedColor
     False Color=&SelectLinearColorFromArray(
      Array Sampling Mode=::NewEnumerator0
      Direct Array Index=?
      Color Selection Array=DataInterface() {propertyValues {ColorData [<rgba 1 0.01 1 1>, <rgba 0.01 1 1 1>, <rgba 1 0.5 0.01 1>, <rgba 0.2 0.01 1 1>]}}
      Random Seed=0
     )
    )
    Scale Color=&Multiply_VectorByFloat(Vector=<1 1 1>, Float=&MakeCustomFloatFromBool(Bool=$User.bIsPainted, True Float=5, False Float=1))
   ) {module=Color001}
   &Drag(Drag=5, Use Linear Drag=true, Use Rotational Drag=true)
   &ParticleState()
   &SolveForcesAndVelocity(Clamp Velocity=false, Limit Acceleration=false)
   &ScaleSpriteSize(Uniform Curve Sprite Scale=DataInterface() {propertyValues {Curve {keys [<at 0 0 RCIM_Cubic>,<at 0.1 1 RCIM_Cubic>,<at 1 0 RCIM_Cubic>]}}})
  ]
  rendererClasses [&NiagaraSpriteRendererProperties]
  renderers [{rendererClass=&NiagaraSpriteRendererProperties}]
 }
 WheelWrap_Particles001 {
  simTarget=GPUComputeSim
  emitterUpdateScript [
   &EmitterState()
   &SpawnRate(SpawnRate=150, Use Spawn Probability=false)
  ]
  particleSpawnScript [
   &VC_PaintedVariantsOverride(
    Black=<rgba 0.15 0.15 0.15 1>
    BurntSienna=<rgba 0.2 0.0833 0 1>
    Cobalt=<rgba 0.05 0.0833 0.25 1>
    Crimson=<rgba 0.5 0.05 0.0622 1>
    DefaultColor=&SelectLinearColorFromArray(
     Array Sampling Mode=::NewEnumerator0
     Direct Array Index=?
     Color Selection Array=DataInterface() {propertyValues {ColorData [<rgba 0.3 0.003 0.3 1>, <rgba 0.004 0.4 0.4 1>, <rgba 0.7 0.35 0.007 1>, <rgba 0.2167 0.075 0.5 1>]}}
     Random Seed=0
    )
    ForestGreen=<rgba 0.0875 0.35 0.1008 1>
    Grey=<rgba 0.21 0.21 0.21 1>
    Pink=<rgba 0.5 0.2 0.3909 1>
    Purple=<rgba 0.1467 0.02 0.4 1>
   )
   &InitializeParticle(
    Color Mode=::NewEnumerator0
    Lifetime Max=0.6
    Lifetime Min=0.7
    Lifetime Mode=::NewEnumerator1
    Mesh Scale Max=<0.05 0.8 0.07>
    Mesh Scale Min=<0.03 0.5 0.05>
    Mesh Scale Mode=::NewEnumerator2
    Position Mode=::NewEnumerator0
    Sprite Rotation Mode=::NewEnumerator2
    Write Material Random=true
   )
   &InitialMeshOrientation(Mesh Orientation Mode=::NewEnumerator7, Rotation=<0 0 0.25>, Rotation Coordinate Space=::NewEnumerator0, Use Rotation Vector=true)
   &PartitionParticles()
   &SkeletalMeshLocation(
    Bone / Socket Index=&Add_Integer(A=$Particles.PartitionParticles.Partition, B=4)
    Bone / Socket Sampling Mode=::NewEnumerator5
    Interpolate Mesh Sampling=false
    MaskSpawnGroup=false
    Mesh Sampling Type=::NewEnumerator6
    Skeletal Mesh=$System.SkeletalMesh
   )
   &AddVelocity(Velocity=&MakeCustomVectorFromBool(
     Bool=$User.bIsFrontendPreview
     True Vector=&RandomRangeVector(
      Minimum=&Multiply_VectorByFloat(Vector=<0 0 6>, Float=4)
      Maximum=&Multiply_VectorByFloat(Vector=<-35 0 -5>, Float=4)
     )
     False Vector=&RandomRangeVector(
      Minimum=&Multiply_VectorByFloat(Vector=<-50 0 0>, Float=10)
      Maximum=&Multiply_VectorByFloat(Vector=<-100 0 0>, Float=10)
     )
    )) {enabled=false}
   &Color(Color=&MakeCustomLinearColorFromBool(
     Bool=$User.bIsPainted
     True Color=&Multiply_LinearColorByFloat(
      Linear Color=$User.GlassColor
      Float=&RandomRangeFloat(Minimum=$User.MinimumGlassColor, Maximum=$User.MaxGlassColor)
     )
     False Color=&SelectLinearColorFromArray(
      Array Sampling Mode=::NewEnumerator0
      Direct Array Index=?
      Color Selection Array=DataInterface() {propertyValues {ColorData [<rgba 1 0.01 1 1>, <rgba 0.01 1 1 1>, <rgba 1 0.5 0.01 1>, <rgba 0.2 0.01 1 1>]}}
      Random Seed=0
     )
    )) {enabled=false}
   /SetVariables(Particles.ScaleNenu:Vector3f=&RandomRangeVector(Minimum=<0.009 0.5 0.05>, Maximum=<0.009 0.8 0.07>)) {isSetParameters=true}
  ]
  particleUpdateScript [
   &FN_DisableTireFX(
    Mask Index Attribute:NiagaraInt32=$Particles.PartitionParticles.Partition
    Mask Index FL=2
    Mask Index FR=3
    Mask Index RL=0
    Mask Index RR=1
   )
   &VC_WheelWrapScale(Wheel Index Attribute:NiagaraInt32=&Add_Integer(A=$Particles.PartitionParticles.Partition, B=2))
   &ShapeLocation(
    Handle Radius=1
    Large Radius=&Multiply_Float(A=50, B=&MakeFloatFromVector(Vector=$Particles.Scale, Channel=::NewEnumerator0))
    MaskSpawnGroup=false
    Non Uniform Scale=<1 1 10>
    Offset=&AddVectorToPosition(
     Position=$Particles.SkeletalMeshLocation.SampledPosition
     Vector=&SelectVectorFromArray(
      Direct Array Index=$Particles.PartitionParticles.Partition
      Vector Selection Array=DataInterface() {propertyValues {FloatData [<0 -2 0>, <0 0 0>]}}
     )
    )
    Offset Coordinate Space=::Local
    Offset Mode=::NewEnumerator0
    Override Seed=false
    Rotation Angle:NiagaraFloat=&Multiply_Float(A=$Emitter.Age, B=2)
    Rotation Axis=<0 1 0>
    Rotation Coordinate Space=::Simulation
    Rotation Mode=::NewEnumerator1
    Shape Primitive=::NewEnumerator3
    Surface Distribution=1
    UseOwnerScale=true
   ) {module=ShapeLocation003}
   &ParticleState()
   &MeshRotationForce(Rotation=<0 0 50>) {enabled=false}
   &SolveForcesAndVelocity(Clamp Velocity=false, Limit Acceleration=false)
   &ScaleSpriteSize(Uniform Curve Sprite Scale=DataInterface() {propertyValues {Curve {keys [{tangentMode=RCTM_User, time=0, value=1.5}, {tangentMode=RCTM_User, time=1, value=0}]}}}) {enabled=false}
   &DynamicMaterialParameters(Module.Index0Param0Write=true, Module.Index0Param1Write=false, Module.Index0Param2Write=false, Module.Index0Param3Write=false)
   &ScaleColor(
    Scale Alpha=20
    Scale RGB=&VectorFromFloat(Value=&RandomRangeFloat(
      Minimum=&FloatFromCurve(FloatCurve=DataInterface() {propertyValues {Curve {keys [<at 0 2>,<at 0.1 1>,<at 0.9 1>,<at 1 2>]}}})
      Maximum=&FloatFromCurve(FloatCurve=DataInterface() {propertyValues {Curve {keys [<at 0 15>,<at 0.1 3>,<at 0.9 3>,<at 1 15>]}}})
     ))
    ScaleA=true
    ScaleRGB=true
   ) {module=ScaleColor001}
   &ScaleColor(
    Scale Alpha=20
    Scale RGB=&MakeCustomVectorFromBool(
     Bool=$User.bIsPainted
     True Vector=&VectorFromFloat(Value=&Multiply_Float(
       A=&RandomRangeFloat(Minimum=$User.GlassColorVar, Maximum=1)
       B=&FloatFromCurve(FloatCurve=DataInterface() {propertyValues {Curve {keys [
            <at 0 1>
            {interpMode=RCIM_Cubic, tangentMode=RCTM_Break, time=0.5, value=1}
            {interpMode=RCIM_Cubic, tangentMode=RCTM_Break, time=1, value=25}
           ]}}})
      ))
     False Vector=&VectorFromFloat(Value=&RandomRangeFloat(
       Minimum=&FloatFromCurve(FloatCurve=DataInterface() {propertyValues {Curve {keys [<at 0 10>,<at 0.1 1>,<at 0.9 1>,<at 1 100>]}}}, Scale Curve=0.6)
       Maximum=&FloatFromCurve(FloatCurve=DataInterface() {propertyValues {Curve {keys [<at 0 5>,<at 0.1 1>,<at 0.9 1>,<at 1 5>]}}}, Scale Curve=0.8)
      ))
    )
    ScaleA=true
    ScaleRGB=true
   ) {enabled=false}
   &CameraOffset()
  ]
  rendererClasses [&NiagaraMeshRendererProperties, &NiagaraSpriteRendererProperties]
  renderers [
   {rendererClass=&NiagaraMeshRendererProperties, rendererIndex=0}
   {rendererClass=&NiagaraSpriteRendererProperties, rendererIndex=1}
  ]
 }
}
summary {
 systemName=NS_DriftTrail_HoloGlass_MP
 emitters [
  {emitterName=Glass, rendererClasses [&NiagaraMeshRendererProperties, &NiagaraSpriteRendererProperties]}
  {emitterName=TriangleExplotion, rendererClasses [&NiagaraSpriteRendererProperties]}
  {emitterName=Core, rendererClasses [&NiagaraSpriteRendererProperties]}
  {
   emitterName=WheelWrap_Particles001
   simTarget=GPUComputeSim
   rendererClasses [&NiagaraMeshRendererProperties, &NiagaraSpriteRendererProperties]
  }
 ]
 userVariables [
  User.bIsFrontendPreview {type=NiagaraBool}
  User.bIsPainted {type=NiagaraBool}
  User.ColorVariantID {type=NiagaraInt32}
  User.CoreColor {type=LinearColor, description="", defaultValue {struct=LinearColor, value=<rgba 1 1 1 1>}}
  User.DTCarOffset {type=Vector3f, description="", defaultValue {struct=Vector3f, value=<0 0 0>}}
  User.GlassColor {type=LinearColor, description="", defaultValue {struct=LinearColor, value=<rgba 1 1 1 1>}}
  User.GlassColorVar {type=NiagaraFloat, defaultValue=1, description=""}
  User.GlowColor {type=LinearColor, description="", defaultValue {struct=LinearColor, value=<rgba 1 1 1 1>}}
  User.MaxGlassColor {type=NiagaraFloat, defaultValue=5, description=""}
  User.MinimumGlassColor {type=NiagaraFloat, defaultValue=0.5, description=""}
  User.offset {type=Vector3f, description="offset from wheels(set through cosmetic template)", defaultValue {struct=Vector3f, value=<0 0 0>}}
  User.PaintedColor {type=LinearColor, description="", defaultValue {struct=LinearColor, value=<rgba 1 1 1 1>}}
  User.PaintedScalar {type=NiagaraFloat, defaultValue=1, description=""}
  User.ParticleColorVar {type=NiagaraFloat, defaultValue=1, description=""}
  User.VehicleFlags {type=NiagaraInt32, defaultValue=806289408}
  User.Wheel_Radius_Array {
   type=NiagaraDataInterfaceArrayFloat
   defaultValue {
    struct=NiagaraExt_VariableValue_DataInterface
    value {
     dataInterface=&self:NiagaraDataInterfaceArrayFloat_1
     dataInterfaceClass=&NiagaraDataInterfaceArrayFloat
    }
   }
  }
  User.Wheel_Width_Array {
   type=NiagaraDataInterfaceArrayFloat
   defaultValue {
    struct=NiagaraExt_VariableValue_DataInterface
    value {
     dataInterface=&self:NiagaraDataInterfaceArrayFloat_0
     dataInterfaceClass=&NiagaraDataInterfaceArrayFloat
    }
   }
  }
  User.WheelTouchingArray {
   type=NiagaraDataInterfaceArrayFloat
   defaultValue {
    struct=NiagaraExt_VariableValue_DataInterface
    value {
     dataInterface=&self:NiagaraDataInterfaceArrayFloat_2
     dataInterfaceClass=&NiagaraDataInterfaceArrayFloat
    }
   }
  }
 ]
}
