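# Masked unlit material for Niagara mesh particles, forming the holographic wheel-wrap layer of the HoloGlass drift trail.
# Three copies of one hologram texture pan along U at different tilings and speeds, summing a red-channel layer with two slower green-channel layers.
# Particle color tints that pattern through a shared emissive-adjustment function.
# Particle alpha scales the clamped pattern, which a rounded sine of V trims to a hard-edged central band as a mask.
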Material /VehicleCosmetics/DriftTrail/DriftTrail_HoloGlass/Material/M_DriftTrail_HoloGlass_WheelWrap
graph {
 properties {
  bAutomaticallySetUsageInEditor=false
  bUsedWithNiagaraMeshParticles=true
  blendMode=BLEND_Masked
  shadingModel=MSM_Unlit
  usedShadingModels=Unlit
 }
 nodes [
  n1 = TextureSample(UVs=Panner(Coordinate=TextureCoordinate() {uTiling=1.5, vTiling=1.2}) {speedX=0.3}) {texture=/DealershipSharedLibrary/FX/Textures/TexturePackage/T_VCVFX_HoloGlass_Hologram}
  MaterialExpressionConstantClamp_2 = Saturate(None=@n1.R)
  n2 = TextureSample(UVs=Panner(Coordinate=TextureCoordinate() {uTiling=0.25}) {speedX=0.05}) {texture=&T_VCVFX_HoloGlass_Hologram}
  n3 = TextureSample(UVs=Panner(Coordinate=TextureCoordinate() {uTiling=0.5}) {speedX=0.02}) {texture=&T_VCVFX_HoloGlass_Hologram}
  n4 = Add(A=@n1.R, B=Multiply(A=Add(A=@n2.G, B=@n3.G), B=Constant() {r=1}))
  Add(A=Add(A=@n1.R, B=@MaterialExpressionConstantClamp_2), B=@n4) {constB=0}
  MaterialExpressionMeshEmitterVertexColor_0 = ParticleColor()
  MaterialFunctionCall(Emissive=Multiply(A=@n4, B=@MaterialExpressionMeshEmitterVertexColor_0.RGB)) {materialFunction=&MF_FX_EmissiveAdjustments}
  MaterialExpressionConstantClamp_1 = Saturate(None=Multiply(A=@n4, B=@MaterialExpressionMeshEmitterVertexColor_0.A))
  n5 = Sine(None=ComponentMask(None=TextureCoordinate()) {g=true}) {period=2}
  Multiply(A=@MaterialExpressionConstantClamp_1, B=Round(None=Power(Base=@n5) {constExponent=0.98}))
 ]
}
